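# Physics asset for the male medium Drift Horror Vest body cosmetic, set up for dynamic left and right collar bones.
# Each collar is a box body joined to a tiny capsule on its MSTR bone by a constraint with tight twist and asymmetric swing limits.
# Head and neck capsules, linked by a constraint with free swing and twist, give the collar pieces body shapes to collide against.
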PhysicsAsset /BRCosmetics/Characters/Player/Male/Medium/Bodies/M_MED_Drift_Horror_Vest/Meshes/M_MED_Drift_Horror_Vest_Physics
bodies {
 dyn_collar_l [{
   shapeType=Box
   extentX=7.795
   extentY=2.4306
   extentZ=12.7679
   shapeName=dyn_collar_l_box
   center=<-4.5 0 1>
   rotation=<rot -21.8859 -26.7967 90>
  }]
 dyn_collar_l_MSTR [{shapeType=Capsule, length=0.505, radius=0.505, shapeName=dyn_collar_l_MSTR_capsule, rotation=<rot 0 -90 0>}]
 dyn_collar_r [{
   shapeType=Box
   extentX=8.5677
   extentY=2.4306
   extentZ=12.7679
   shapeName=dyn_collar_r_box
   center=<-4.5 0 1.5>
   rotation=<rot 17.6931 -152.6907 89.9992>
  }]
 dyn_collar_r_MSTR [{shapeType=Capsule, length=0.505, radius=0.505, shapeName=dyn_collar_r_MSTR_capsule, rotation=<rot 0 -90 0>}]
 head [{shapeType=Capsule, length=7.6663, radius=5.3282, shapeName=head_capsule, center=<-2.0457 8.6414 -0.0001>, rotation=<rot 0 90 90>}]
 neck_02 [{shapeType=Capsule, length=6.3566, radius=6.0332, shapeName=neck_02_capsule, center=<3.4946 -1.856 -0.1394>, rotation=<rot 90 0 7.7168>}]
}
constraints [
 {bone1Name=dyn_collar_l, bone2Name=dyn_collar_l_MSTR, swing1LimitDegrees=15, swing2LimitDegrees=45, twistLimitDegrees=5}
 {bone1Name=dyn_collar_r, bone2Name=dyn_collar_r_MSTR, swing1LimitDegrees=15, swing2LimitDegrees=45, twistLimitDegrees=5}
 {
  bone1Name=head
  bone2Name=neck_02
  swing1LimitDegrees=45
  swing1Motion=Free
  swing2LimitDegrees=45
  swing2Motion=Free
  twistLimitDegrees=45
  twistMotion=Free
 }
]
