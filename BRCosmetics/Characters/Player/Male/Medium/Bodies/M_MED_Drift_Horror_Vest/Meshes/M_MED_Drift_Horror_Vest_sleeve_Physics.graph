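# Physics asset for the sleeves of the M_MED_Drift_Horror_Vest player body cosmetic, giving dangling sleeve bones secondary motion.
# Capsules on both lower arms anchor four left and three right chains, each a tiny MSTR capsule bone followed by a box-shaped sleeve bone.
# MSTR-to-arm constraints have free swing and twist, while sleeve-to-MSTR constraints carry swing limits and a tight twist limit, narrower in swing on left chain 03.
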PhysicsAsset /BRCosmetics/Characters/Player/Male/Medium/Bodies/M_MED_Drift_Horror_Vest/Meshes/M_MED_Drift_Horror_Vest_sleeve_Physics
bodies {
 dyn_sleeve_l_01 [{
   shapeType=Box
   extentX=2.0841
   extentY=1.7226
   extentZ=6.7931
   shapeName=dyn_sleeve_l_01_box
   center=<6.5613 -0.1434 -0.3197>
   rotation=<rot 0 95.0409 -84.6532>
  }]
 dyn_sleeve_l_01_MSTR [{shapeType=Capsule, length=0.505, radius=0.505, shapeName=dyn_sleeve_l_01_MSTR_capsule, rotation=<rot 0 -90 0>}]
 dyn_sleeve_l_02 [{
   shapeType=Box
   extentX=1.903
   extentY=1.3432
   extentZ=3.4993
   shapeName=dyn_sleeve_l_02_box
   center=<5.0723 -0.2866 -0.0919>
   rotation=<rot 0 116.9736 -84.8682>
  }]
 dyn_sleeve_l_02_MSTR [{shapeType=Capsule, length=0.505, radius=0.505, shapeName=dyn_sleeve_l_02_MSTR_capsule, rotation=<rot 0 -90 0>}]
 dyn_sleeve_l_03 [{
   shapeType=Box
   extentX=2.4958
   extentY=1.4107
   extentZ=4.9984
   shapeName=dyn_sleeve_l_03_box
   center=<5.4832 -0.1001 0.6551>
   rotation=<rot 0 -84.6477 -77.7853>
  }]
 dyn_sleeve_l_03_MSTR [{shapeType=Capsule, length=0.505, radius=0.505, shapeName=dyn_sleeve_l_03_MSTR_capsule, rotation=<rot 0 -90 0>}]
 dyn_sleeve_l_04 [{
   shapeType=Box
   extentX=2.0352
   extentY=1.4599
   extentZ=3.1818
   shapeName=dyn_sleeve_l_04_box
   center=<3.5561 -0.1192 0.0692>
   rotation=<rot 0 -67.7208 -88.7125>
  }]
 dyn_sleeve_l_04_MSTR [{shapeType=Capsule, length=0.505, radius=0.505, shapeName=dyn_sleeve_l_04_MSTR_capsule, rotation=<rot 0 -90 0>}]
 dyn_sleeve_r_01 [{
   shapeType=Box
   extentX=2.0593
   extentY=2.0667
   extentZ=3.9301
   shapeName=dyn_sleeve_r_01_box
   center=<5.0839 -0.7203 -0.261>
   rotation=<rot 0 109.0769 -73.8918>
  }]
 dyn_sleeve_r_01_MSTR [{shapeType=Capsule, length=0.505, radius=0.505, shapeName=dyn_sleeve_r_01_MSTR_capsule, rotation=<rot 0 -90 0>}]
 dyn_sleeve_r_02 [{
   shapeType=Box
   extentX=2.9874
   extentY=1.5111
   extentZ=7.2529
   shapeName=dyn_sleeve_r_02_box
   center=<7.0654 0.5262 -0.7604>
   rotation=<rot 0 82.9339 -80.9143>
  }]
 dyn_sleeve_r_02_MSTR [{shapeType=Capsule, length=0.505, radius=0.505, shapeName=dyn_sleeve_r_02_MSTR_capsule, rotation=<rot 0 -90 0>}]
 dyn_sleeve_r_03 [{
   shapeType=Box
   extentX=2.119
   extentY=1.9065
   extentZ=3.8681
   shapeName=dyn_sleeve_r_03_box
   center=<5.4128 0.5143 -0.1722>
   rotation=<rot 0 71.6437 -84.6176>
  }]
 dyn_sleeve_r_03_MSTR [{shapeType=Capsule, length=0.505, radius=0.505, shapeName=dyn_sleeve_r_03_MSTR_capsule, rotation=<rot 0 -90 0>}]
 lowerarm_l [{shapeType=Capsule, length=15.0642, radius=3.9358, center=<9 0 0.5>, rotation=<rot 0 90 90>}]
 lowerarm_r [{shapeType=Capsule, length=14.9281, radius=4.0719, center=<-9 0 -0.25>, rotation=<rot 0 90 90>}]
}
constraints [
 {
  bone1Name=dyn_sleeve_l_01_MSTR
  bone2Name=lowerarm_l
  swing1LimitDegrees=45
  swing1Motion=Free
  swing2LimitDegrees=45
  swing2Motion=Free
  twistLimitDegrees=45
  twistMotion=Free
 }
 {bone1Name=dyn_sleeve_l_01, bone2Name=dyn_sleeve_l_01_MSTR, swing1LimitDegrees=45, swing2LimitDegrees=45, twistLimitDegrees=5}
 {
  bone1Name=dyn_sleeve_l_02_MSTR
  bone2Name=lowerarm_l
  swing1LimitDegrees=45
  swing1Motion=Free
  swing2LimitDegrees=45
  swing2Motion=Free
  twistLimitDegrees=45
  twistMotion=Free
 }
 {bone1Name=dyn_sleeve_l_02, bone2Name=dyn_sleeve_l_02_MSTR, swing1LimitDegrees=45, swing2LimitDegrees=45, twistLimitDegrees=5}
 {
  bone1Name=dyn_sleeve_l_03_MSTR
  bone2Name=lowerarm_l
  swing1LimitDegrees=45
  swing1Motion=Free
  swing2LimitDegrees=45
  swing2Motion=Free
  twistLimitDegrees=45
  twistMotion=Free
 }
 {bone1Name=dyn_sleeve_l_03, bone2Name=dyn_sleeve_l_03_MSTR, swing1LimitDegrees=20, swing2LimitDegrees=35, twistLimitDegrees=5}
 {
  bone1Name=dyn_sleeve_l_04_MSTR
  bone2Name=lowerarm_l
  swing1LimitDegrees=45
  swing1Motion=Free
  swing2LimitDegrees=45
  swing2Motion=Free
  twistLimitDegrees=45
  twistMotion=Free
 }
 {bone1Name=dyn_sleeve_l_04, bone2Name=dyn_sleeve_l_04_MSTR, swing1LimitDegrees=45, swing2LimitDegrees=45, twistLimitDegrees=5}
 {
  bone1Name=dyn_sleeve_r_01_MSTR
  bone2Name=lowerarm_r
  swing1LimitDegrees=45
  swing1Motion=Free
  swing2LimitDegrees=45
  swing2Motion=Free
  twistLimitDegrees=45
  twistMotion=Free
 }
 {bone1Name=dyn_sleeve_r_01, bone2Name=dyn_sleeve_r_01_MSTR, swing1LimitDegrees=45, swing2LimitDegrees=45, twistLimitDegrees=5}
 {
  bone1Name=dyn_sleeve_r_02_MSTR
  bone2Name=lowerarm_r
  swing1LimitDegrees=45
  swing1Motion=Free
  swing2LimitDegrees=45
  swing2Motion=Free
  twistLimitDegrees=45
  twistMotion=Free
 }
 {bone1Name=dyn_sleeve_r_02, bone2Name=dyn_sleeve_r_02_MSTR, swing1LimitDegrees=45, swing2LimitDegrees=45, twistLimitDegrees=5}
 {
  bone1Name=dyn_sleeve_r_03_MSTR
  bone2Name=lowerarm_r
  swing1LimitDegrees=45
  swing1Motion=Free
  swing2LimitDegrees=45
  swing2Motion=Free
  twistLimitDegrees=45
  twistMotion=Free
 }
 {bone1Name=dyn_sleeve_r_03, bone2Name=dyn_sleeve_r_03_MSTR, swing1LimitDegrees=45, swing2LimitDegrees=45, twistLimitDegrees=5}
]
