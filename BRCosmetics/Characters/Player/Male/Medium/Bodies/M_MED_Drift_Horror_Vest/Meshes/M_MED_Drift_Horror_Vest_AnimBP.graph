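# Post-pose AnimBP for the M_MED_Drift_Horror_Vest player body cosmetic, adding deformation and secondary motion to the pose copied from the attached parent mesh.
# Emperor twist and sculpter Control Rigs feed weighted constraints that position collar, zipper, shirt, boot and sleeve bones.
# Partial-alpha AnimDynamics then swing shirt, boots and zippers, followed by two RigidBody passes using vest and sleeve physics assets.
# Drift Horror Control Rigs and per-bone sleeve AnimDynamics sit unconnected to the output.
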AnimBlueprint /BRCosmetics/Characters/Player/Male/Medium/Bodies/M_MED_Drift_Horror_Vest/Meshes/M_MED_Drift_Horror_Vest_AnimBP
target_skeleton=&/M_MED_Drift_Horror_Vest_Skeleton
graphs [
 AnimGraph {nodes [
   n1 = Anim Dynamics - Bone: dyn_sleeve_l_01(Alpha=1, ExternalForce="0,0,0") {
    angularDampingOverride=0.9999
    angularSpringConstant=100
    bAngularSpring=true
    bGravityOverrideInSimSpace=true
    bOverrideAngularDamping=true
    bUsePlanarLimit=false
    simulationSpace=RootRelative
    alphaBoolBlend {blendOption=Linear}
    boundBone {boneName=dyn_sleeve_l_01}
    componentAppliedLinearAccClamp=<3500 3500 3500>
    componentLinearAccScale=<0.9 0.9 0.9>
    componentLinearVelScale=<0.9 0.9 0.9>
    physicsBodyDefinitions [{
      collisionType=CustomSphere
      sphereCollisionRadius=4
      boundBone {boneName=dyn_sleeve_l_01}
      boxExtents=<15 10 10>
      constraintSetup {angularLimitsMax=<10 35 10>, angularLimitsMin=<-10 -35 -30>, angularTarget=<0 0 0>}
      localJointOffset=<3 0 0>
     }]
    relativeSpaceBone {boneName=root}
    sphericalLimits [{limitRadius=6, drivingBone {boneName=upperarm_twist_02_l}, sphereLocalOffset=<0 0 -3>}]
   }
   n2 = Anim Dynamics - Bone: dyn_sleeve_l_02(ComponentPose=@n1, Alpha=1, ExternalForce="0,0,0") {
    angularDampingOverride=0.9999
    angularSpringConstant=100
    bAngularSpring=true
    bGravityOverrideInSimSpace=true
    bOverrideAngularDamping=true
    bUsePlanarLimit=false
    simulationSpace=RootRelative
    alphaBoolBlend {blendOption=Linear}
    boundBone {boneName=dyn_sleeve_l_02}
    componentAppliedLinearAccClamp=<3500 3500 3500>
    componentLinearAccScale=<0.9 0.9 0.9>
    componentLinearVelScale=<0.9 0.9 0.9>
    physicsBodyDefinitions [{
      collisionType=CustomSphere
      sphereCollisionRadius=4
      boundBone {boneName=dyn_sleeve_l_02}
      boxExtents=<15 10 10>
      constraintSetup {angularLimitsMax=<10 0 25>, angularLimitsMin=<-10 -35 -25>, angularTarget=<0 0 0>}
      localJointOffset=<3 0 0>
     }]
    relativeSpaceBone {boneName=root}
    sphericalLimits [{limitRadius=6, drivingBone {boneName=upperarm_twist_02_l}, sphereLocalOffset=<0 0 -3>}]
   }
   n3 = Anim Dynamics - Bone: dyn_sleeve_l_03(ComponentPose=@n2, Alpha=1, ExternalForce="0,0,0") {
    angularDampingOverride=0.9999
    angularSpringConstant=100
    bAngularSpring=true
    bGravityOverrideInSimSpace=true
    bOverrideAngularDamping=true
    bUsePlanarLimit=false
    simulationSpace=RootRelative
    alphaBoolBlend {blendOption=Linear}
    boundBone {boneName=dyn_sleeve_l_03}
    componentAppliedLinearAccClamp=<3500 3500 3500>
    componentLinearAccScale=<0.9 0.9 0.9>
    componentLinearVelScale=<0.9 0.9 0.9>
    physicsBodyDefinitions [{
      collisionType=CustomSphere
      sphereCollisionRadius=4
      boundBone {boneName=dyn_sleeve_l_03}
      boxExtents=<15 10 10>
      constraintSetup {angularLimitsMax=<10 5 25>, angularLimitsMin=<-10 -35 -25>, angularTarget=<0 0 0>}
      localJointOffset=<3 0 0>
     }]
    relativeSpaceBone {boneName=root}
    sphericalLimits [{limitRadius=6, drivingBone {boneName=upperarm_twist_02_l}, sphereLocalOffset=<0 0 -3>}]
   }
   n4 = Anim Dynamics - Bone: dyn_sleeve_l_04(ComponentPose=@n3, Alpha=1, ExternalForce="0,0,0") {
    angularDampingOverride=0.9999
    angularSpringConstant=100
    bAngularSpring=true
    bGravityOverrideInSimSpace=true
    bOverrideAngularDamping=true
    bUsePlanarLimit=false
    simulationSpace=RootRelative
    alphaBoolBlend {blendOption=Linear}
    boundBone {boneName=dyn_sleeve_l_04}
    componentAppliedLinearAccClamp=<3500 3500 3500>
    componentLinearAccScale=<0.9 0.9 0.9>
    componentLinearVelScale=<0.9 0.9 0.9>
    physicsBodyDefinitions [{
      collisionType=CustomSphere
      sphereCollisionRadius=4
      boundBone {boneName=dyn_sleeve_l_04}
      boxExtents=<15 10 10>
      constraintSetup {angularLimitsMax=<10 15 25>, angularLimitsMin=<-10 -35 -25>, angularTarget=<0 0 0>}
      localJointOffset=<3 0 0>
     }]
    relativeSpaceBone {boneName=root}
    sphericalLimits [{limitRadius=6, drivingBone {boneName=upperarm_twist_02_l}, sphereLocalOffset=<0 0 -3>}]
   }
   n5 = Anim Dynamics - Bone: dyn_sleeve_r_01(ComponentPose=@n4, Alpha=1, ExternalForce="0,0,0") {
    angularDampingOverride=0.9999
    angularSpringConstant=100
    bAngularSpring=true
    bGravityOverrideInSimSpace=true
    bOverrideAngularDamping=true
    bUsePlanarLimit=false
    simulationSpace=RootRelative
    alphaBoolBlend {blendOption=Linear}
    boundBone {boneName=dyn_sleeve_r_01}
    componentAppliedLinearAccClamp=<3500 3500 3500>
    componentLinearAccScale=<0.9 0.9 0.9>
    componentLinearVelScale=<0.9 0.9 0.9>
    physicsBodyDefinitions [{
      collisionType=CustomSphere
      sphereCollisionRadius=4
      boundBone {boneName=dyn_sleeve_r_01}
      boxExtents=<15 10 10>
      constraintSetup {angularLimitsMax=<10 35 25>, angularLimitsMin=<-10 -35 -10>, angularTarget=<0 0 0>}
      localJointOffset=<3 0 0>
     }]
    relativeSpaceBone {boneName=root}
    sphericalLimits [{limitRadius=6, drivingBone {boneName=upperarm_twist_02_l}, sphereLocalOffset=<0 0 -3>}]
   }
   n6 = Anim Dynamics - Bone: dyn_sleeve_r_02(ComponentPose=@n5, Alpha=1, ExternalForce="0,0,0") {
    angularDampingOverride=0.9999
    angularSpringConstant=100
    bAngularSpring=true
    bGravityOverrideInSimSpace=true
    bOverrideAngularDamping=true
    bUsePlanarLimit=false
    simulationSpace=RootRelative
    alphaBoolBlend {blendOption=Linear}
    boundBone {boneName=dyn_sleeve_r_02}
    componentAppliedLinearAccClamp=<3500 3500 3500>
    componentLinearAccScale=<0.9 0.9 0.9>
    componentLinearVelScale=<0.9 0.9 0.9>
    physicsBodyDefinitions [{
      collisionType=CustomSphere
      sphereCollisionRadius=4
      boundBone {boneName=dyn_sleeve_r_02}
      boxExtents=<15 10 10>
      constraintSetup {angularLimitsMax=<10 10 25>, angularLimitsMin=<-10 -35 -5>, angularTarget=<0 0 0>}
      localJointOffset=<3 0 0>
     }]
    relativeSpaceBone {boneName=root}
    sphericalLimits [{limitRadius=6, drivingBone {boneName=upperarm_twist_02_l}, sphereLocalOffset=<0 0 -3>}]
   }
   Anim Dynamics - Bone: dyn_sleeve_r_03(ComponentPose=@n6, Alpha=1, ExternalForce="0,0,0") {
    angularDampingOverride=0.9999
    angularSpringConstant=100
    bAngularSpring=true
    bGravityOverrideInSimSpace=true
    bOverrideAngularDamping=true
    bUsePlanarLimit=false
    simulationSpace=RootRelative
    alphaBoolBlend {blendOption=Linear}
    boundBone {boneName=dyn_sleeve_r_03}
    componentAppliedLinearAccClamp=<3500 3500 3500>
    componentLinearAccScale=<0.9 0.9 0.9>
    componentLinearVelScale=<0.9 0.9 0.9>
    physicsBodyDefinitions [{
      collisionType=CustomSphere
      sphereCollisionRadius=4
      boundBone {boneName=dyn_sleeve_r_03}
      boxExtents=<15 10 10>
      constraintSetup {angularLimitsMax=<10 15 25>, angularLimitsMin=<-10 -35 -25>, angularTarget=<0 0 0>}
      localJointOffset=<3 0 0>
     }]
    relativeSpaceBone {boneName=root}
    sphericalLimits [{limitRadius=6, drivingBone {boneName=upperarm_twist_02_l}, sphereLocalOffset=<0 0 -3>}]
   }
   Control Rig() {
    alphaBoolBlend {blendOption=Linear}
    controlRigAssetReference {blueprintRigClass=/BRCosmetics/Characters/Player/Male/Medium/Bodies/M_MED_Drift_Horror/Meshes/M_MED_Drift_Horror_CtrlRig.M_MED_Drift_Horror_CtrlRig_C}
   }
   n7 = Control Rig() {
    alphaBoolBlend {blendOption=Linear}
    controlRigAssetReference {blueprintRigClass=/Game/Characters/Player/Male/Medium/Bodies/M_MED_Drift_Horror/Meshes/M_MED_Drift_Horror_Deform_CtrlRig_twists_2.M_MED_Drift_Horror_Deform_CtrlRig_twists_2_C}
   }
   Control Rig!(
    Source=@n7
    Alpha
    KneeRot_fr_l_114deg:Vector
    KneeTrans_fr_l_114deg:Vector="0.258,-3.591,0.009"
    KneeScale_fr_l_114deg:Vector="1,1,1"
    KneeScale_fr_l_130deg:Vector="1,1,1"
    KneeRot_fr_l_130deg:Vector="0,0,-51.2"
    KneeTrans_fr_l_130deg:Vector="0.258,-3.591,0.009"
   ) -> (Pose) {
    alphaBoolBlend {blendOption=Linear}
    controlRigAssetReference {blueprintRigClass=/Game/Characters/Player/Male/Medium/Bodies/M_MED_Drift_Horror/Meshes/M_MED_drift_horror_Deform_CtrlRig_sculpters_2.M_MED_drift_horror_Deform_CtrlRig_sculpters_2_C}
   }
   n8 = Control Rig!(Source=Copy Pose From Mesh() {bUseAttachedParent=true}, Alpha) -> (Pose) {
    alphaBoolBlend {blendOption=Linear}
    controlRigAssetReference {blueprintRigClass=/Game/Characters/Player/Male/Medium/Bodies/M_MED_Emperor/Meshes/M_MED_Emperor_Deform_CtrlRig_twists.M_MED_Emperor_Deform_CtrlRig_twists_C}
   }
   n9 = Control Rig!(
    Source=@n8
    Alpha
    KneeRot_fr_l_114deg:Vector
    KneeTrans_fr_l_114deg:Vector="0.258,-3.591,0.009"
    KneeScale_fr_l_114deg:Vector="1,1,1"
    KneeScale_fr_l_130deg:Vector="1,1,1"
    KneeRot_fr_l_130deg:Vector="0,0,-51.2"
    KneeTrans_fr_l_130deg:Vector="0.258,-3.591,0.009"
   ) -> (Pose) {
    alphaBoolBlend {blendOption=Linear}
    controlRigAssetReference {blueprintRigClass=/Game/Characters/Player/Male/Medium/Bodies/M_MED_Emperor/Meshes/M_MED_Emperor_Deform_CtrlRig_sculpters.M_MED_Emperor_Deform_CtrlRig_sculpters_C}
   }
   n10 = "Constraint - dyn_zipper_MSTR (PPPPPPP)"!(
    ComponentPose=Local To Component(LocalPose=@n9)
    Alpha=1
    ConstraintWeights_0=0.0092
    ConstraintWeights_1:float=0.001
    ConstraintWeights_2:float=0.014
    ConstraintWeights_3:float=0.7015
    ConstraintWeights_4:float=0.0013
    ConstraintWeights_5:float=0.1256
    ConstraintWeights_6:float=0.1472
   ) -> (Pose) {
    alphaBoolBlend {blendOption=Linear}
    boneToModify {boneName=dyn_zipper_MSTR}
    constraintSetup [
     {targetBone {boneName=clavicle_l}}
     {targetBone {boneName=spine_03}}
     {targetBone {boneName=spine_04}}
     {targetBone {boneName=spine_05}}
     {targetBone {boneName=clavicle_r}}
     {targetBone {boneName=neck_02}}
     {targetBone {boneName=neck_01}}
    ]
    constraintWeights [1,1,1,1,1,1,1]
   }
   n11 = "Constraint - dyn_zipper_02_MSTR (PPPPPPPP)"!(
    ComponentPose=@n10
    Alpha=1
    ConstraintWeights_0=0.0014
    ConstraintWeights_1:float=0.0328
    ConstraintWeights_2:float=0.0061
    ConstraintWeights_3:float=0.023
    ConstraintWeights_4:float=0.0579
    ConstraintWeights_5:float=0.4053
    ConstraintWeights_6:float=0.4416
    ConstraintWeights_7:float=0.0319
   ) -> (Pose) {
    alphaBoolBlend {blendOption=Linear}
    boneToModify {boneName=dyn_zipper_02_MSTR}
    constraintSetup [
     {targetBone {boneName=upperarm_twist_01_r}}
     {targetBone {boneName=deform_pectoral_fr_r}}
     {targetBone {boneName=deform_deltoid_fr_r}}
     {targetBone {boneName=spine_02}}
     {targetBone {boneName=spine_03}}
     {targetBone {boneName=spine_04}}
     {targetBone {boneName=spine_05}}
     {targetBone {boneName=clavicle_r}}
    ]
    constraintWeights [1,1,1,1,1,1,1,1]
   }
   n12 = "Constraint - dyn_collar_l_MSTR (PPPPPPPP)"!(
    ComponentPose=@n11
    Alpha=1
    ConstraintWeights_0=0.0507
    ConstraintWeights_1:float=0.0042
    ConstraintWeights_2:float=0.0024
    ConstraintWeights_3:float=0.0248
    ConstraintWeights_4:float=0.8475
    ConstraintWeights_5:float=0.0023
    ConstraintWeights_6:float=0.0304
    ConstraintWeights_7:float=0.0376
   ) -> (Pose) {
    alphaBoolBlend {blendOption=Linear}
    boneToModify {boneName=dyn_collar_l_MSTR}
    constraintSetup [
     {targetBone {boneName=clavicle_l}}
     {targetBone {boneName=deform_pectoral_fr_l}}
     {targetBone {boneName=spine_03}}
     {targetBone {boneName=spine_04}}
     {targetBone {boneName=spine_05}}
     {targetBone {boneName=clavicle_r}}
     {targetBone {boneName=neck_02}}
     {targetBone {boneName=neck_01}}
    ]
    constraintWeights [1,1,1,1,1,1,1,1]
   }
   n13 = "Constraint - dyn_collar_r_MSTR (PPPPPPPP)"!(
    ComponentPose=@n12
    Alpha=1
    ConstraintWeights_0=0.0061
    ConstraintWeights_1:float=0.0014
    ConstraintWeights_2:float=0.0032
    ConstraintWeights_3:float=0.0269
    ConstraintWeights_4:float=0.8524
    ConstraintWeights_5:float=0.0291
    ConstraintWeights_6:float=0.0377
    ConstraintWeights_7:float=0.0432
   ) -> (Pose) {
    alphaBoolBlend {blendOption=Linear}
    boneToModify {boneName=dyn_collar_r_MSTR}
    constraintSetup [
     {targetBone {boneName=clavicle_l}}
     {targetBone {boneName=deform_pectoral_fr_r}}
     {targetBone {boneName=spine_03}}
     {targetBone {boneName=spine_04}}
     {targetBone {boneName=spine_05}}
     {targetBone {boneName=clavicle_r}}
     {targetBone {boneName=neck_02}}
     {targetBone {boneName=neck_01}}
    ]
    constraintWeights [1,1,1,1,1,1,1,1]
   }
   n14 = "Constraint - dyn_shirt_a_MSTR (PPPPPPPP)"!(
    ComponentPose=@n13
    Alpha=1
    ConstraintWeights_0=0.028
    ConstraintWeights_1:float=0.0051
    ConstraintWeights_2:float=0.146
    ConstraintWeights_3:float=0.0399
    ConstraintWeights_4:float=0.0229
    ConstraintWeights_5:float=0.6521
    ConstraintWeights_6:float=0.0691
    ConstraintWeights_7:float=0.037
   ) -> (Pose) {
    alphaBoolBlend {blendOption=Linear}
    boneToModify {boneName=dyn_shirt_a_MSTR}
    constraintSetup [
     {targetBone {boneName=deform_thigh_fr_l}}
     {targetBone {boneName=thigh_twist_01_r}}
     {targetBone {boneName=spine_01}}
     {targetBone {boneName=spine_02}}
     {targetBone {boneName=spine_03}}
     {targetBone {boneName=pelvis}}
     {targetBone {boneName=thigh_twist_01_l}}
     {targetBone {boneName=deform_thigh_bk_l}}
    ]
    constraintWeights [1,1,1,1,1,1,1,1]
   }
   n15 = "Constraint - dyn_shirt_b_MSTR (PPPPPPPP)"!(
    ComponentPose=@n14
    Alpha=1
    ConstraintWeights_0=0.0153
    ConstraintWeights_1:float=0.0173
    ConstraintWeights_2:float=0.1457
    ConstraintWeights_3:float=0.0243
    ConstraintWeights_4:float=0.0161
    ConstraintWeights_5:float=0.6927
    ConstraintWeights_6:float=0.0429
    ConstraintWeights_7:float=0.0456
   ) -> (Pose) {
    alphaBoolBlend {blendOption=Linear}
    boneToModify {boneName=dyn_shirt_b_MSTR}
    constraintSetup [
     {targetBone {boneName=deform_thigh_bk_r}}
     {targetBone {boneName=thigh_twist_01_r}}
     {targetBone {boneName=spine_01}}
     {targetBone {boneName=spine_02}}
     {targetBone {boneName=spine_03}}
     {targetBone {boneName=pelvis}}
     {targetBone {boneName=thigh_twist_01_l}}
     {targetBone {boneName=deform_thigh_bk_l}}
    ]
    constraintWeights [1,1,1,1,1,1,1,1]
   }
   n16 = "Constraint - dyn_shirt_c_MSTR (PPPPPPPP)"!(
    ComponentPose=@n15
    Alpha=1
    ConstraintWeights_0=0.04
    ConstraintWeights_1:float=0.0312
    ConstraintWeights_2:float=0.1531
    ConstraintWeights_3:float=0.0221
    ConstraintWeights_4:float=0.0157
    ConstraintWeights_5:float=0.6783
    ConstraintWeights_6:float=0.0272
    ConstraintWeights_7:float=0.0324
   ) -> (Pose) {
    alphaBoolBlend {blendOption=Linear}
    boneToModify {boneName=dyn_shirt_c_MSTR}
    constraintSetup [
     {targetBone {boneName=deform_thigh_bk_r}}
     {targetBone {boneName=thigh_twist_01_r}}
     {targetBone {boneName=spine_01}}
     {targetBone {boneName=spine_02}}
     {targetBone {boneName=spine_03}}
     {targetBone {boneName=pelvis}}
     {targetBone {boneName=thigh_twist_01_l}}
     {targetBone {boneName=deform_thigh_bk_l}}
    ]
    constraintWeights [1,1,1,1,1,1,1,1]
   }
   n17 = "Constraint - dyn_shirt_d_MSTR (PPPPPPPP)"!(
    ComponentPose=@n16
    Alpha=1
    ConstraintWeights_0=0.043
    ConstraintWeights_1:float=0.0442
    ConstraintWeights_2:float=0.1499
    ConstraintWeights_3:float=0.0291
    ConstraintWeights_4:float=0.0188
    ConstraintWeights_5:float=0.6904
    ConstraintWeights_6:float=0.0138
    ConstraintWeights_7:float=0.0108
   ) -> (Pose) {
    alphaBoolBlend {blendOption=Linear}
    boneToModify {boneName=dyn_shirt_d_MSTR}
    constraintSetup [
     {targetBone {boneName=deform_thigh_bk_r}}
     {targetBone {boneName=thigh_twist_01_r}}
     {targetBone {boneName=spine_01}}
     {targetBone {boneName=spine_02}}
     {targetBone {boneName=spine_03}}
     {targetBone {boneName=pelvis}}
     {targetBone {boneName=thigh_twist_01_l}}
     {targetBone {boneName=deform_thigh_bk_l}}
    ]
    constraintWeights [1,1,1,1,1,1,1,1]
   }
   n18 = "Constraint - dyn_shirt_e_MSTR (PPPPPPPP)"!(
    ComponentPose=@n17
    Alpha=1
    ConstraintWeights_0=0.0354
    ConstraintWeights_1:float=0.0578
    ConstraintWeights_2:float=0.1513
    ConstraintWeights_3:float=0.0411
    ConstraintWeights_4:float=0.0241
    ConstraintWeights_5:float=0.664
    ConstraintWeights_6:float=0.0058
    ConstraintWeights_7:float=0.0206
   ) -> (Pose) {
    alphaBoolBlend {blendOption=Linear}
    boneToModify {boneName=dyn_shirt_e_MSTR}
    constraintSetup [
     {targetBone {boneName=deform_thigh_bk_r}}
     {targetBone {boneName=thigh_twist_01_r}}
     {targetBone {boneName=spine_01}}
     {targetBone {boneName=spine_02}}
     {targetBone {boneName=spine_03}}
     {targetBone {boneName=pelvis}}
     {targetBone {boneName=thigh_twist_01_l}}
     {targetBone {boneName=deform_thigh_fr_r}}
    ]
    constraintWeights [1,1,1,1,1,1,1,1]
   }
   n19 = "Constraint - dyn_shirt_f_MSTR (PPPPPPPP)"!(
    ComponentPose=@n18
    Alpha=1
    ConstraintWeights_0=0.0351
    ConstraintWeights_1:float=0.091
    ConstraintWeights_2:float=0.1483
    ConstraintWeights_3:float=0.0363
    ConstraintWeights_4:float=0.019
    ConstraintWeights_5:float=0.0025
    ConstraintWeights_6:float=0.6064
    ConstraintWeights_7:float=0.0613
   ) -> (Pose) {
    alphaBoolBlend {blendOption=Linear}
    boneToModify {boneName=dyn_shirt_f_MSTR}
    constraintSetup [
     {targetBone {boneName=deform_thigh_bk_r}}
     {targetBone {boneName=thigh_twist_01_r}}
     {targetBone {boneName=spine_01}}
     {targetBone {boneName=spine_02}}
     {targetBone {boneName=spine_03}}
     {targetBone {boneName=spine_04}}
     {targetBone {boneName=pelvis}}
     {targetBone {boneName=deform_thigh_fr_r}}
    ]
    constraintWeights [1,1,1,1,1,1,1,1]
   }
   n20 = "Constraint - dyn_boot_l_MSTR (PPPP)"!(
    ComponentPose=@n19
    Alpha=1
    ConstraintWeights_0=0.0129
    ConstraintWeights_1:float=0.5734
    ConstraintWeights_2:float=0.1423
    ConstraintWeights_3:float=0.2714
   ) -> (Pose) {
    alphaBoolBlend {blendOption=Linear}
    boneToModify {boneName=dyn_boot_l_MSTR}
    constraintSetup [
     {targetBone {boneName=calf_l}}
     {targetBone {boneName=calf_twist_01_l}}
     {targetBone {boneName=foot_l}}
     {targetBone {boneName=calf_twist_02_l}}
    ]
    constraintWeights [1,1,1,1]
   }
   n21 = "Constraint - dyn_boot_r_MSTR (PPPP)"!(
    ComponentPose=@n20
    Alpha=1
    ConstraintWeights_0=0.2543
    ConstraintWeights_1:float=0.5804
    ConstraintWeights_2:float=0.154
    ConstraintWeights_3:float=0.0113
   ) -> (Pose) {
    alphaBoolBlend {blendOption=Linear}
    boneToModify {boneName=dyn_boot_r_MSTR}
    constraintSetup [
     {targetBone {boneName=calf_twist_02_r}}
     {targetBone {boneName=calf_twist_01_r}}
     {targetBone {boneName=foot_r}}
     {targetBone {boneName=calf_r}}
    ]
    constraintWeights [1,1,1,1]
   }
   n22 = "Constraint - dyn_collar (PPPPPP)"!(
    ComponentPose=@n21
    Alpha=1
    ConstraintWeights_0=0.0148
    ConstraintWeights_1:float=0.0048
    ConstraintWeights_2:float=0.2692
    ConstraintWeights_3:float=0.0244
    ConstraintWeights_4:float=0.3436
    ConstraintWeights_5:float=0.3432
   ) -> (Pose) {
    alphaBoolBlend {blendOption=Linear}
    boneToModify {boneName=dyn_collar}
    constraintSetup [
     {targetBone {boneName=clavicle_l}}
     {targetBone {boneName=spine_03}}
     {targetBone {boneName=spine_05}}
     {targetBone {boneName=clavicle_r}}
     {targetBone {boneName=neck_02}}
     {targetBone {boneName=neck_01}}
    ]
    constraintWeights [1,1,1,1,1,1]
   }
   n23 = "Constraint - dyn_sleeve_l_03_MSTR (PPPPP)"!(
    ComponentPose=@n22
    Alpha=1
    ConstraintWeights_0=0.2341
    ConstraintWeights_1:float=0.0735
    ConstraintWeights_2:float=0.1374
    ConstraintWeights_3:float=0.4951
    ConstraintWeights_4:float=0.0599
   ) -> (Pose) {
    alphaBoolBlend {blendOption=Linear}
    boneToModify {boneName=dyn_sleeve_l_03_MSTR}
    constraintSetup [
     {targetBone {boneName=lowerarm_l}}
     {targetBone {boneName=lowerarm_twist_01_l}}
     {targetBone {boneName=deform_elbow_fr_l}}
     {targetBone {boneName=lowerarm_twist_02_l}}
     {targetBone {boneName=deform_elbow_bk_l}}
    ]
    constraintWeights [1,1,1,1,1]
   }
   n24 = "Constraint - dyn_sleeve_l_04_MSTR (PPPPP)"!(
    ComponentPose=@n23
    Alpha=1
    ConstraintWeights_0=0.1949
    ConstraintWeights_1:float=0.0749
    ConstraintWeights_2:float=0.0542
    ConstraintWeights_3:float=0.3965
    ConstraintWeights_4:float=0.2795
   ) -> (Pose) {
    alphaBoolBlend {blendOption=Linear}
    boneToModify {boneName=dyn_sleeve_l_04_MSTR}
    constraintSetup [
     {targetBone {boneName=lowerarm_l}}
     {targetBone {boneName=lowerarm_twist_01_l}}
     {targetBone {boneName=deform_elbow_fr_l}}
     {targetBone {boneName=lowerarm_twist_02_l}}
     {targetBone {boneName=deform_elbow_bk_l}}
    ]
    constraintWeights [1,1,1,1,1]
   }
   n25 = "Constraint - dyn_sleeve_l_02_MSTR (PPPPPPP)"!(
    ComponentPose=@n24
    Alpha=1
    ConstraintWeights_0=0.0013
    ConstraintWeights_1:float=0.2033
    ConstraintWeights_2:float=0.0876
    ConstraintWeights_3:float=0.1405
    ConstraintWeights_4:float=0.0011
    ConstraintWeights_5:float=0.5347
    ConstraintWeights_6:float=0.0314
   ) -> (Pose) {
    alphaBoolBlend {blendOption=Linear}
    boneToModify {boneName=dyn_sleeve_l_02_MSTR}
    constraintSetup [
     {targetBone {boneName=upperarm_twist_02_l}}
     {targetBone {boneName=lowerarm_l}}
     {targetBone {boneName=lowerarm_twist_01_l}}
     {targetBone {boneName=deform_elbow_fr_l}}
     {targetBone {boneName=upperarm_l}}
     {targetBone {boneName=lowerarm_twist_02_l}}
     {targetBone {boneName=deform_elbow_bk_l}}
    ]
    constraintWeights [1,1,1,1,1,1,1]
   }
   n26 = "Constraint - dyn_sleeve_l_01_MSTR (PPPPP)"!(
    ComponentPose=@n25
    Alpha=1
    ConstraintWeights_0=0.1972
    ConstraintWeights_1:float=0.0773
    ConstraintWeights_2:float=0.1026
    ConstraintWeights_3:float=0.4586
    ConstraintWeights_4:float=0.1643
   ) -> (Pose) {
    alphaBoolBlend {blendOption=Linear}
    boneToModify {boneName=dyn_sleeve_l_01_MSTR}
    constraintSetup [
     {targetBone {boneName=lowerarm_l}}
     {targetBone {boneName=lowerarm_twist_01_l}}
     {targetBone {boneName=deform_elbow_fr_l}}
     {targetBone {boneName=lowerarm_twist_02_l}}
     {targetBone {boneName=deform_elbow_bk_l}}
    ]
    constraintWeights [1,1,1,1,1]
   }
   n27 = "Constraint - dyn_sleeve_r_01_MSTR (PPPPPP)"!(
    ComponentPose=@n26
    Alpha=1
    ConstraintWeights_0=0.4591
    ConstraintWeights_1:float=0.1594
    ConstraintWeights_2:float=0.1094
    ConstraintWeights_3:float=0.1987
    ConstraintWeights_4:float=0.0704
    ConstraintWeights_5:float=0.003
   ) -> (Pose) {
    alphaBoolBlend {blendOption=Linear}
    boneToModify {boneName=dyn_sleeve_r_01_MSTR}
    constraintSetup [
     {targetBone {boneName=lowerarm_twist_02_r}}
     {targetBone {boneName=deform_elbow_bk_r}}
     {targetBone {boneName=deform_elbow_fr_r}}
     {targetBone {boneName=lowerarm_r}}
     {targetBone {boneName=lowerarm_twist_01_r}}
     {targetBone {boneName=upperarm_twist_01_r}}
    ]
    constraintWeights [1,1,1,1,1,1]
   }
   n28 = "Constraint - dyn_sleeve_r_03_MSTR (PPPPPPPP)"!(
    ComponentPose=@n27
    Alpha=1
    ConstraintWeights_0=0.5234
    ConstraintWeights_1:float=0.0458
    ConstraintWeights_2:float=0.0041
    ConstraintWeights_3:float=0.1272
    ConstraintWeights_4:float=0.2003
    ConstraintWeights_5:float=0.0891
    ConstraintWeights_6:float=0.0037
    ConstraintWeights_7:float=0.0065
   ) -> (Pose) {
    alphaBoolBlend {blendOption=Linear}
    boneToModify {boneName=dyn_sleeve_r_03_MSTR}
    constraintSetup [
     {targetBone {boneName=lowerarm_twist_02_r}}
     {targetBone {boneName=deform_elbow_bk_r}}
     {targetBone {boneName=upperarm_r}}
     {targetBone {boneName=deform_elbow_fr_r}}
     {targetBone {boneName=lowerarm_r}}
     {targetBone {boneName=lowerarm_twist_01_r}}
     {targetBone {boneName=upperarm_twist_01_r}}
     {targetBone {boneName=upperarm_twist_02_r}}
    ]
    constraintWeights [1,1,1,1,1,1,1,1]
   }
   n29 = "Constraint - dyn_sleeve_r_02_MSTR (PPPPPPPP)"!(
    ComponentPose=@n28
    Alpha=1
    ConstraintWeights_0=0.2959
    ConstraintWeights_1:float=0.2424
    ConstraintWeights_2:float=0.0496
    ConstraintWeights_3:float=0.1371
    ConstraintWeights_4:float=0.1699
    ConstraintWeights_5:float=0.025
    ConstraintWeights_6:float=0.011
    ConstraintWeights_7:float=0.069
   ) -> (Pose) {
    alphaBoolBlend {blendOption=Linear}
    boneToModify {boneName=dyn_sleeve_r_02_MSTR}
    constraintSetup [
     {targetBone {boneName=lowerarm_twist_02_r}}
     {targetBone {boneName=deform_elbow_bk_r}}
     {targetBone {boneName=upperarm_r}}
     {targetBone {boneName=deform_elbow_fr_r}}
     {targetBone {boneName=lowerarm_r}}
     {targetBone {boneName=lowerarm_twist_01_r}}
     {targetBone {boneName=upperarm_twist_01_r}}
     {targetBone {boneName=upperarm_twist_02_r}}
    ]
    constraintWeights [1,1,1,1,1,1,1,1]
   }
   n30 = Anim Dynamics - Bone: dyn_shirt_a!(ComponentPose=@n29, Alpha=0.5, GravityOverride:Vector="0,0,0", ExternalForce="0,0,0") -> (Pose) {
    angularDampingOverride=0.9999
    angularSpringConstant=1500
    bAngularSpring=true
    bGravityOverrideInSimSpace=true
    bOverrideAngularDamping=true
    gravityScale=0.8
    simulationSpace=BoneRelative
    alphaBoolBlend {blendOption=Linear}
    boundBone {boneName=dyn_shirt_a}
    componentAppliedLinearAccClamp=<5000 5000 3500>
    componentLinearAccScale=<1.5 1.5 1.25>
    componentLinearVelScale=<1.5 1.5 1.25>
    physicsBodyDefinitions [{
      sphereCollisionRadius=0
      boundBone {boneName=dyn_shirt_a}
      boxExtents=<10 20 20>
      constraintSetup {angularLimitsMax=<15 5 15>, angularLimitsMin=<-15 -75 -15>, angularTarget=<0 0 0>}
      localJointOffset=<2 0 0>
     }]
    relativeSpaceBone {boneName=root}
   }
   n31 = Anim Dynamics - Bone: dyn_shirt_b!(ComponentPose=@n30, Alpha=0.5, GravityOverride:Vector="0,0,0", ExternalForce="0,0,0") -> (Pose) {
    angularDampingOverride=0.9999
    angularSpringConstant=1500
    bAngularSpring=true
    bGravityOverrideInSimSpace=true
    bOverrideAngularDamping=true
    gravityScale=0.8
    simulationSpace=BoneRelative
    alphaBoolBlend {blendOption=Linear}
    boundBone {boneName=dyn_shirt_b}
    componentAppliedLinearAccClamp=<5000 5000 3500>
    componentLinearAccScale=<1.5 1.5 1.25>
    componentLinearVelScale=<1.5 1.5 1.25>
    physicsBodyDefinitions [{
      sphereCollisionRadius=0
      boundBone {boneName=dyn_shirt_b}
      boxExtents=<10 20 20>
      constraintSetup {angularLimitsMax=<15 5 15>, angularLimitsMin=<-15 -75 -15>, angularTarget=<0 0 0>}
      localJointOffset=<2 0 0>
     }]
    relativeSpaceBone {boneName=root}
   }
   n32 = Anim Dynamics - Bone: dyn_shirt_c!(ComponentPose=@n31, Alpha=0.5, GravityOverride:Vector="0,0,0", ExternalForce="0,0,0") -> (Pose) {
    angularDampingOverride=0.9999
    angularSpringConstant=1500
    bAngularSpring=true
    bGravityOverrideInSimSpace=true
    bOverrideAngularDamping=true
    gravityScale=0.8
    simulationSpace=BoneRelative
    alphaBoolBlend {blendOption=Linear}
    boundBone {boneName=dyn_shirt_c}
    componentAppliedLinearAccClamp=<5000 5000 3500>
    componentLinearAccScale=<1.5 1.5 1.25>
    componentLinearVelScale=<1.5 1.5 1.25>
    physicsBodyDefinitions [{
      sphereCollisionRadius=0
      boundBone {boneName=dyn_shirt_c}
      boxExtents=<10 20 20>
      constraintSetup {angularLimitsMax=<15 5 15>, angularLimitsMin=<-15 -90 -15>, angularTarget=<0 0 0>}
      localJointOffset=<2 0 0>
     }]
    relativeSpaceBone {boneName=root}
   }
   n33 = Anim Dynamics - Bone: dyn_shirt_d!(ComponentPose=@n32, Alpha=0.5, GravityOverride:Vector="0,0,0", ExternalForce="0,0,0") -> (Pose) {
    angularDampingOverride=0.9999
    angularSpringConstant=1500
    bAngularSpring=true
    bGravityOverrideInSimSpace=true
    bOverrideAngularDamping=true
    gravityScale=0.8
    simulationSpace=BoneRelative
    alphaBoolBlend {blendOption=Linear}
    boundBone {boneName=dyn_shirt_d}
    componentAppliedLinearAccClamp=<5000 5000 3500>
    componentLinearAccScale=<1.5 1.5 1.25>
    componentLinearVelScale=<1.5 1.5 1.25>
    physicsBodyDefinitions [{
      sphereCollisionRadius=0
      boundBone {boneName=dyn_shirt_d}
      boxExtents=<10 20 20>
      constraintSetup {angularLimitsMax=<15 5 15>, angularLimitsMin=<-15 -90 -15>, angularTarget=<0 0 0>}
      localJointOffset=<2 0 0>
     }]
    relativeSpaceBone {boneName=root}
   }
   n34 = Anim Dynamics - Bone: dyn_shirt_e!(ComponentPose=@n33, Alpha=0.5, GravityOverride:Vector="0,0,0", ExternalForce="0,0,0") -> (Pose) {
    angularDampingOverride=0.9999
    angularSpringConstant=1500
    bAngularSpring=true
    bGravityOverrideInSimSpace=true
    bOverrideAngularDamping=true
    gravityScale=0.8
    simulationSpace=BoneRelative
    alphaBoolBlend {blendOption=Linear}
    boundBone {boneName=dyn_shirt_e}
    componentAppliedLinearAccClamp=<5000 5000 3500>
    componentLinearAccScale=<1.5 1.5 1.25>
    componentLinearVelScale=<1.5 1.5 1.25>
    physicsBodyDefinitions [{
      sphereCollisionRadius=0
      boundBone {boneName=dyn_shirt_e}
      boxExtents=<10 20 20>
      constraintSetup {angularLimitsMax=<15 5 15>, angularLimitsMin=<-15 -75 -15>, angularTarget=<0 0 0>}
      localJointOffset=<2 0 0>
     }]
    relativeSpaceBone {boneName=root}
   }
   n35 = Anim Dynamics - Bone: dyn_shirt_f!(ComponentPose=@n34, Alpha=0.5, GravityOverride:Vector="0,0,0", ExternalForce="0,0,0") -> (Pose) {
    angularDampingOverride=0.9999
    angularSpringConstant=1500
    bAngularSpring=true
    bGravityOverrideInSimSpace=true
    bOverrideAngularDamping=true
    gravityScale=0.8
    simulationSpace=BoneRelative
    alphaBoolBlend {blendOption=Linear}
    boundBone {boneName=dyn_shirt_f}
    componentAppliedLinearAccClamp=<5000 5000 3500>
    componentLinearAccScale=<1.5 1.5 1.25>
    componentLinearVelScale=<1.5 1.5 1.25>
    physicsBodyDefinitions [{
      sphereCollisionRadius=0
      boundBone {boneName=dyn_shirt_f}
      boxExtents=<10 20 20>
      constraintSetup {angularLimitsMax=<15 5 15>, angularLimitsMin=<-15 -75 -15>, angularTarget=<0 0 0>}
      localJointOffset=<2 0 0>
     }]
    relativeSpaceBone {boneName=root}
   }
   n36 = Anim Dynamics - Bone: dyn_boot_l!(ComponentPose=@n35, Alpha=0.5, GravityOverride:Vector="0,0,0", ExternalForce="0,0,0") -> (Pose) {
    angularDampingOverride=0.9999
    angularSpringConstant=1500
    bAngularSpring=true
    bGravityOverrideInSimSpace=true
    bOverrideAngularDamping=true
    gravityScale=0.8
    simulationSpace=BoneRelative
    alphaBoolBlend {blendOption=Linear}
    boundBone {boneName=dyn_boot_l}
    componentAppliedLinearAccClamp=<5000 5000 3500>
    componentLinearAccScale=<1.5 1.5 1.25>
    componentLinearVelScale=<1.5 1.5 1.25>
    physicsBodyDefinitions [{
      sphereCollisionRadius=0
      boundBone {boneName=dyn_boot_l}
      boxExtents=<10 20 20>
      constraintSetup {angularLimitsMax=<15 35 15>, angularLimitsMin=<-15 -35 -15>, angularTarget=<0 0 0>}
      localJointOffset=<2 0 0>
     }]
    relativeSpaceBone {boneName=root}
   }
   n37 = Anim Dynamics - Bone: dyn_boot_r!(ComponentPose=@n36, Alpha=0.5, GravityOverride:Vector="0,0,0", ExternalForce="0,0,0") -> (Pose) {
    angularDampingOverride=0.9999
    angularSpringConstant=1500
    bAngularSpring=true
    bGravityOverrideInSimSpace=true
    bOverrideAngularDamping=true
    gravityScale=0.8
    simulationSpace=BoneRelative
    alphaBoolBlend {blendOption=Linear}
    boundBone {boneName=dyn_boot_r}
    componentAppliedLinearAccClamp=<5000 5000 3500>
    componentLinearAccScale=<1.5 1.5 1.25>
    componentLinearVelScale=<1.5 1.5 1.25>
    physicsBodyDefinitions [{
      sphereCollisionRadius=0
      boundBone {boneName=dyn_boot_r}
      boxExtents=<10 20 20>
      constraintSetup {angularLimitsMax=<15 35 15>, angularLimitsMin=<-15 -35 -15>, angularTarget=<0 0 0>}
      localJointOffset=<2 0 0>
     }]
    relativeSpaceBone {boneName=root}
   }
   n38 = Anim Dynamics - Bone: dyn_zipper!(ComponentPose=@n37, Alpha=0.25, GravityOverride:Vector="0,0,0", ExternalForce) -> (Pose) {
    angularDampingOverride=1
    angularSpringConstant=1500
    bAngularSpring=true
    bGravityOverrideInSimSpace=true
    bOverrideAngularDamping=true
    gravityScale=0.45
    simulationSpace=RootRelative
    alphaBoolBlend {blendOption=Linear}
    alphaScaleBias {scale=2}
    boundBone {boneName=dyn_zipper}
    componentAppliedLinearAccClamp=<2500 2500 2500>
    componentLinearAccScale=<1 1 1>
    componentLinearVelScale=<1 1 1>
    physicsBodyDefinitions [{
      sphereCollisionRadius=0
      boundBone {boneName=dyn_zipper}
      boxExtents=<10 5 5>
      constraintSetup {angularLimitsMax=<15 35 15>, angularLimitsMin=<-15 -35 -15>, angularTarget=<0 0 0>}
      localJointOffset=<2 0 0>
     }]
    relativeSpaceBone {boneName=root}
   }
   n39 = Anim Dynamics - Bone: dyn_zipper_02!(ComponentPose=@n38, Alpha=0.25, GravityOverride:Vector="0,0,0", ExternalForce="0,0,0") -> (Pose) {
    angularDampingOverride=1
    angularSpringConstant=1500
    bAngularSpring=true
    bGravityOverrideInSimSpace=true
    bOverrideAngularDamping=true
    gravityScale=0.45
    simulationSpace=RootRelative
    alphaBoolBlend {blendOption=Linear}
    alphaScaleBias {scale=2}
    boundBone {boneName=dyn_zipper_02}
    componentAppliedLinearAccClamp=<2500 2500 2500>
    componentLinearAccScale=<1 1 1>
    componentLinearVelScale=<1 1 1>
    physicsBodyDefinitions [{
      sphereCollisionRadius=0
      boundBone {boneName=dyn_zipper_02}
      boxExtents=<10 5 5>
      constraintSetup {angularLimitsMax=<15 35 15>, angularLimitsMin=<-15 -35 -15>, angularTarget=<0 0 0>}
      localJointOffset=<2 0 0>
     }]
    relativeSpaceBone {boneName=root}
   }
   n40 = RigidBody(ComponentPose=@n39) {
    bClampLinearTranslationLimitToRefPose=true
    bOverrideWorldGravity=true
    overridePhysicsAsset=&/M_MED_Drift_Horror_Vest_Physics
    simulationSpace=BaseBoneSpace
    alphaBoolBlend {blendOption=Linear}
    baseBoneRef {boneName=spine_05}
    componentAppliedLinearAccClamp=<3000 3000 3000>
    componentLinearAccScale=<0.3 0.3 0.3>
    componentLinearVelScale=<0.3 0.3 0.3>
    overrideWorldGravity {z=-980}
   }
   n41 = RigidBody(ComponentPose=@n40) {
    bClampLinearTranslationLimitToRefPose=true
    bOverrideWorldGravity=true
    overridePhysicsAsset=&/M_MED_Drift_Horror_Vest_sleeve_Physics
    simulationSpace=BaseBoneSpace
    alphaBoolBlend {blendOption=Linear}
    baseBoneRef {boneName=root}
    componentAppliedLinearAccClamp=<3000 3000 3000>
    componentLinearAccScale=<0.3 0.3 0.3>
    componentLinearVelScale=<0.3 0.3 0.3>
    overrideWorldGravity {z=-980}
   }
   Output Pose(Result=Component To Local(ComponentPose=@n41)) {layerGroup=None}
  ]}
 EventGraph {nodes [
   Event BlueprintInitializeAnimation()
   > Set Skel_Mesh(Skel_Mesh=GetOwnerSkeletalMeshComponent())
   Event BlueprintUpdateAnimation()
  ]}
]
variables [Skel_Mesh {type=SkeletalMeshComponent}]
