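# Post-pose AnimBP for the Drift Horror Vest face accessory, adding secondary motion to its hood and ears.
# It copies the attached parent's pose, then layers a copy from the owner skeletal mesh component, cached at initialization, over collar, deltoid and pectoral bones.
# A RigidBody node simulates the hood physics asset relative to spine_05.
# Chained AnimDynamics nodes then add damped, angle-limited spring motion to each ear bone.
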AnimBlueprint /BRCosmetics/Characters/Player/Male/Medium/Bodies/M_MED_Drift_Horror_Vest/Meshes/Parts/M_MED_Drift_Horror_Vest_FaceAcc_AnimBP
target_skeleton=&/M_MED_Drift_Horror_Vest_FaceAcc_Skeleton
graphs [
 AnimGraph {nodes [
   n1 = Layered blend per bone(
    BasePose=Copy Pose From Mesh() {bUseAttachedParent=true}
    BlendPoses_0=Copy Pose From Mesh(SourceMeshComponent=Get Skel_Mesh()) {bUseAttachedParent=true}
   ) {layerSetup [{branchFilters [
       {boneName=dyn_collar_l}
       {boneName=dyn_collar_r}
       {boneName=deform_deltoid_fr_r}
       {boneName=deform_pectoral_fr_r}
       {boneName=deform_deltoid_fr_l}
       {boneName=deform_pectoral_fr_l}
      ]}]}
   n2 = RigidBody(ComponentPose=Local To Component(LocalPose=@n1)) {
    bClampLinearTranslationLimitToRefPose=true
    bOverrideWorldGravity=true
    overridePhysicsAsset=&/M_MED_Drift_Horror_Vest_Hood_Physics
    simulationSpace=BaseBoneSpace
    alphaBoolBlend {blendOption=Linear}
    baseBoneRef {boneName=spine_05}
    componentAppliedLinearAccClamp=<3000 3000 3000>
    componentLinearAccScale=<0.3 0.3 0.3>
    componentLinearVelScale=<0.3 0.3 0.3>
    overrideWorldGravity {z=-980}
   }
   n3 = Anim Dynamics - Bone: dyn_ear_l!(ComponentPose=@n2, Alpha=1, GravityOverride:Vector="0,0,0", ExternalForce="0,0,0") -> (Pose) {
    angularDampingOverride=0.9999
    angularSpringConstant=1500
    bAngularSpring=true
    bGravityOverrideInSimSpace=true
    bOverrideAngularDamping=true
    gravityScale=0.8
    simulationSpace=BoneRelative
    alphaBoolBlend {blendOption=Linear}
    boundBone {boneName=dyn_ear_l}
    componentAppliedLinearAccClamp=<5000 5000 3500>
    componentLinearAccScale=<1.5 1.5 1.25>
    componentLinearVelScale=<1.5 1.5 1.25>
    physicsBodyDefinitions [{
      sphereCollisionRadius=0
      boundBone {boneName=dyn_ear_l}
      boxExtents=<10 20 20>
      constraintSetup {angularLimitsMax=<15 35 15>, angularLimitsMin=<-15 -35 -15>, angularTarget=<0 0 0>}
      localJointOffset=<2 0 0>
     }]
    relativeSpaceBone {boneName=root}
   }
   n4 = Anim Dynamics - Bone: dyn_ear_r!(ComponentPose=@n3, Alpha=1, GravityOverride:Vector="0,0,0", ExternalForce="0,0,0") -> (Pose) {
    angularDampingOverride=0.9999
    angularSpringConstant=1500
    bAngularSpring=true
    bGravityOverrideInSimSpace=true
    bOverrideAngularDamping=true
    gravityScale=0.8
    simulationSpace=BoneRelative
    alphaBoolBlend {blendOption=Linear}
    boundBone {boneName=dyn_ear_r}
    componentAppliedLinearAccClamp=<5000 5000 3500>
    componentLinearAccScale=<1.5 1.5 1.25>
    componentLinearVelScale=<1.5 1.5 1.25>
    physicsBodyDefinitions [{
      sphereCollisionRadius=0
      boundBone {boneName=dyn_ear_r}
      boxExtents=<10 20 20>
      constraintSetup {angularLimitsMax=<15 35 15>, angularLimitsMin=<-15 -35 -15>, angularTarget=<0 0 0>}
      localJointOffset=<2 0 0>
     }]
    relativeSpaceBone {boneName=root}
   }
   Output Pose(Result=Component To Local(ComponentPose=@n4)) {layerGroup=None}
  ]}
 EventGraph {nodes [
   Event BlueprintInitializeAnimation()
   > Set Skel_Mesh(Skel_Mesh=GetOwnerSkeletalMeshComponent())
   Event BlueprintUpdateAnimation()
  ]}
]
variables [Skel_Mesh {type=SkeletalMeshComponent}]
